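# Demo material for the M_Med_MeteorMan player body, a simple Default Lit surface enabled for skeletal meshes, morph targets and mesh deformers.
# An AO texture parameter, defaulting to a white Facebook logo, feeds two multiplies: one by the Value scalar, one by a constant of one.
# Rough and Metalness scalars and a flat-default Normal map parameter complete the node set, with no material output connections shown.
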Material /Game/Characters/Player/Male/Medium/Bodies/M_Med_MeteorMan/Demo/Materials/DemoMaterial
graph {
 properties {
  bUsedWithMeshDeformer=true
  bUsedWithMorphTargets=true
  bUsedWithSkeletalMesh=true
  pixelDepthOffsetMode=PDOM_Legacy
  usedShadingModels=Default Lit
 }
 nodes [
  n1 = TextureSampleParameter2D() {parameterName=AO, texture=/Game/UI/Frontend/Lobby/SocialImport/Textures/FB-f-Logo__white_29}
  Multiply(A=ScalarParameter() {defaultValue=0.5, parameterName=Value}, B=@n1.RGB)
  Multiply(A=Constant() {r=1}, B=@n1.RGB)
  ScalarParameter() {defaultValue=0.5, parameterName=Rough}
  ScalarParameter() {defaultValue=1, parameterName=Metalness}
  TextureSampleParameter2D() {parameterName=Normal, samplerType=SAMPLERTYPE_Normal, texture=&FlatNormalBlue}
 ]
}
